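# Master material for weapon scopes, kept with the RedDotCircle scope assets, Default Lit and flagged for Nanite.
# A Normal texture parameter is transformed to world space, as tangent-space normals are off, and feeds the shared AttachScope material function.
# BaseColor and SpecularMasks texture parameters default to black textures, with no connections shown from them or into the material outputs.
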Material /Game/Athena/Items/Weapons/Scopes/RedDotCircle/Materials/M_GlobalScope_Master
graph {
 properties {bTangentSpaceNormal=false, bUsedWithNanite=true, pixelDepthOffsetMode=PDOM_Legacy, usedShadingModels=Default Lit}
 nodes [
  MaterialFunctionCall!(
   RotateDialMask
   Normal WS=Transform(None=TextureSampleParameter2D() {parameterName=Normal, samplerType=SAMPLERTYPE_Normal, texture=&FlatNormalBlue}.RGB)
  ) {materialFunction=/Game/Packages/Fortress_SharedMaterials/Base_Material_Functions/AttachScope}
  TextureSampleParameter2D() {parameterName=BaseColor, texture=&T_BlacksRGB}
  TextureSampleParameter2D() {parameterName=SpecularMasks, samplerType=SAMPLERTYPE_Masks, texture=&T_BlackMask}
 ]
}
